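# Sound cue for a looping communication console terminal hum, an indoor ambient spot sound for the Society Estate POI.
# A quality-level node routes every tier to one looping wave player, which plays the Loop_04 wave despite the cue's 05 name.
# Inline attenuation overrides the assigned attenuation asset, adding occlusion and listener focus.
# Concurrency is also overridden, capping simultaneous instances per owner.
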SoundCue /Game/Sounds/Helios/POI/SocietyEstate/Spots/Spot_Console_Terminal_Communication_Hum_Loop_05_Cue
root=QualityLevel_0
nodes [
 {id=QualityLevel_0, children [WavePlayer_1,WavePlayer_1,WavePlayer_1,WavePlayer_1]}
 {
  id=WavePlayer_1
  properties {bLooping=true, soundWaveAssetPtr=/Game/Sounds/Helios/Environment_Actors/ConsoleTerminals/Spot_Console_Terminal_Communication_Hum_Loop_04}
 }
]
properties {
 attenuationSettings=/Game/Sounds/Helios/Environment_Actors/ConsoleTerminals/Spot_Console_Terminal_Communication_Hum_Att
 bOverrideAttenuation=true
 bOverrideConcurrency=true
 maxDistance=2500
 soundClassObject=&Fort_Ambient_Indoors_Athena
 subtitlePriority=5000
 volumeMultiplier=0.877
 attenuationOverrides {
  bEnableFocusInterpolation=true
  bEnableListenerFocus=true
  bEnableOcclusion=true
  dBAttenuationAtMax=-40
  falloffDistance=1700
  focusAzimuth=15
  focusVolumeAttenuation=1.5
  nonFocusAzimuth=45
  nonFocusVolumeAttenuation=0.5
  occlusionInterpolationTime=1
  occlusionLowPassFilterFrequency=10000
  occlusionTraceChannel=ECC_GameTraceChannel11
  occlusionVolumeAttenuation=0.5
  stereoSpread=512
  attenuationShapeExtents=<800 64.5942 90>
 }
 concurrencyOverrides {bLimitToOwner=true, maxCount=2}
}
