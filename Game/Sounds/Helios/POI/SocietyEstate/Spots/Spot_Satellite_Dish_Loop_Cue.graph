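# Looping outdoor ambient spot sound for a satellite dish at the Society Estate POI, reusing the Yacht radar antenna loop.
# The looping wave passes through a neutral Modulator and an Enveloper that ramps volume up from silence, then into Attenuation.
# Attenuation overrides a borrowed outpost windmill asset, adding occlusion and listener focus.
# The cue plays quietly at slightly raised pitch, using Athena prop concurrency.
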SoundCue /Game/Sounds/Helios/POI/SocietyEstate/Spots/Spot_Satellite_Dish_Loop_Cue
root=Attenuation_2
nodes [
 {
  id=Attenuation_2
  children [Enveloper_1]
  properties {
   attenuationSettings=/Game/Sounds/Artemis/Biomes/Grasslands/Grassland_Spots/Spot_Outpost_Windmill_Close_Stereo_Att
   bOverrideAttenuation=true
   attenuationOverrides {
    bEnableListenerFocus=true
    bEnableOcclusion=true
    coneSphereFalloffDistance=35.9442
    coneSphereRadius=600.6739
    falloffDistance=1000
    nonFocusVolumeAttenuation=0.75
    occlusionLowPassFilterFrequency=1000
    occlusionTraceChannel=ECC_GameTraceChannel11
    occlusionVolumeAttenuation=0.5
    stereoSpread=120
    attenuationShapeExtents=<200 -670.4398 -598.4461>
   }
  }
 }
 {
  id=Enveloper_1
  children [Modulator_1]
  properties {pitchCurve {editorCurveData {keys [{time=0}]}}, volumeCurve {editorCurveData {keys [<at 0 0>,<at 1 1>]}}}
 }
 {id=Modulator_1, children [WavePlayer_2], properties {pitchMax=1, pitchMin=1, volumeMax=1, volumeMin=1}}
 {id=WavePlayer_2, properties {bLooping=true, soundWaveAssetPtr=/Game/Sounds/Helios/POI/Yacht/Spots/Spot_Yacht_Radar_Antenna_Loop}}
]
properties {
 bEnableBusSends=false
 maxDistance=1200
 pitchMultiplier=1.1
 soundClassObject=&Fort_Ambient_Outdoors_Athena
 volumeMultiplier=0.2
 concurrencyOverrides {maxCount=2}
 concurrencySet [/Game/Sounds/Concurrency_Rules/Athena_Prop]
 preEffectBusSends [{}]
}
